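# Sound cue for dropping an Impulse Grenade, using the shared weapon-drop attenuation and the loot-collect sound class.
# A mixer layers a random pick of two drop-layer waves, a quieter random pick of three impact waves, and an even quieter fixed fourth impact wave.
# Both impact branches get random pitch at or below normal, while an outer modulator randomly pitches the whole mix up for variation.
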SoundCue /Game/Sounds/Fort_GamePlay_Sounds/Explosives/ImpulseGrenade/SC_ImpulseGrenade_Drop
root=Modulator_0
nodes [
 {id=Mixer_0, children [Random_0,Modulator_1,Modulator_2], properties {inputVolume [1.2,0.6,0.35]}}
 {id=Modulator_0, children [Mixer_0], properties {pitchMax=1.2, pitchMin=1.1}}
 {id=Modulator_1, children [Random_5], properties {pitchMax=1, pitchMin=0.8}}
 {id=Modulator_2, children [WavePlayer_14], properties {pitchMax=1, pitchMin=0.9}}
 {id=Random_0, children [WavePlayer_15, WavePlayer_16], properties {weights [1,1]}}
 {id=Random_5, children [WavePlayer_10, WavePlayer_17, WavePlayer_18], properties {weights [1,1,1]}}
 {id=WavePlayer_10, properties {soundWaveAssetPtr=&/ImpulseGrenade_DropImpact_01}}
 {id=WavePlayer_14, properties {soundWaveAssetPtr=&/ImpulseGrenade_DropImpact_04}}
 {id=WavePlayer_15, properties {soundWaveAssetPtr=&/ImpulseGrenade_DropLayer_01}}
 {id=WavePlayer_16, properties {soundWaveAssetPtr=&/ImpulseGrenade_DropLayer_02}}
 {id=WavePlayer_17, properties {soundWaveAssetPtr=&/ImpulseGrenade_DropImpact_03}}
 {id=WavePlayer_18, properties {soundWaveAssetPtr=&ImpulseGrenade_DropImpact_04}}
]
properties {
 attenuationSettings=&Fort_Weapon_Drop_Att
 maxDistance=3000
 pitchMultiplier=1.1
 soundClassObject=&Fort_Loot_Collect
 volumeMultiplier=1
 attenuationOverrides {falloffDistance=2400, stereoSpread=2000, attenuationShapeExtents {x=800}}
}
