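# Sound cue for dropping an Impulse Grenade, played through the loot-collect sound class.
# A random node picks one of three equally weighted drop wave recordings, and a modulator shifts its pitch slightly upward for variation.
# It is spatialized with the shared weapon-drop attenuation asset and carries a boosted volume multiplier.
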SoundCue /Game/Sounds/Fort_GamePlay_Sounds/Explosives/ImpulseGrenade/SC_ImpulseGrenade_Drop_Pre
root=Modulator_0
nodes [
 {id=Modulator_0, children [Random_0], properties {pitchMax=1.2, pitchMin=1.1}}
 {id=Random_0, children [WavePlayer_0,WavePlayer_1,WavePlayer_2], properties {weights [1,1,1]}}
 {id=WavePlayer_0, properties {soundWaveAssetPtr=&/ImpulseGrenade_Drop_01}}
 {id=WavePlayer_1, properties {soundWaveAssetPtr=&/ImpulseGrenade_Drop_02}}
 {id=WavePlayer_2, properties {soundWaveAssetPtr=&/ImpulseGrenade_Drop_03}}
]
properties {
 attenuationSettings=&Fort_Weapon_Drop_Att
 maxDistance=3000
 soundClassObject=&Fort_Loot_Collect
 volumeMultiplier=1.3
 attenuationOverrides {falloffDistance=2400, stereoSpread=2000, attenuationShapeExtents {x=800}}
}
