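# Sound cue for the Impulse Grenade explosion, blending close and distant layers.
# A mixer combines both layers, each with its own explosion attenuation and random pitch and volume modulation, with the close layer weighted louder in the mix.
# Quality-level switches pick randomly from four waves per layer on the first two levels and only two on the last two.
# A concurrency override prevents new instances once the limit is reached.
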SoundCue /Game/Sounds/Fort_GamePlay_Sounds/Explosives/ImpulseGrenade/SC_ImpulseGrenade_Explosion
root=Mixer_2
nodes [
 {id=Attenuation_1, children [Modulator_5], properties {attenuationSettings=/Game/Sounds/Attenuation/Fort_Explosion_Medium_Athena_Grenade_Att}}
 {id=Attenuation_2, children [Modulator_4], properties {attenuationSettings=/Game/Sounds/Attenuation/Fort_Explosion_Close_Athena_Grenade_Att}}
 {id=Mixer_2, children [Attenuation_2, Attenuation_1], properties {inputVolume [1.35,1]}}
 {id=Modulator_4, children [QualityLevel_0], properties {pitchMax=1, pitchMin=0.9, volumeMax=1.3, volumeMin=1.2}}
 {id=Modulator_5, children [QualityLevel_1], properties {pitchMax=0.9, pitchMin=0.8, volumeMax=1.7, volumeMin=1.6}}
 {id=QualityLevel_0, children [Random_4,Random_4,Random_0,Random_0]}
 {id=QualityLevel_1, children [Random_5,Random_5,Random_1,Random_1]}
 {id=Random_0, children [WavePlayer_8,WavePlayer_9], properties {weights [1,1]}}
 {id=Random_1, children [WavePlayer_18, WavePlayer_19], properties {weights [1,1]}}
 {id=Random_4, children [WavePlayer_8, WavePlayer_9, WavePlayer_10, WavePlayer_17], properties {weights [1,1,1,1]}}
 {id=Random_5, children [WavePlayer_18, WavePlayer_19, WavePlayer_20, WavePlayer_21], properties {weights [1,1,1,1]}}
 {
  id=WavePlayer_10
  properties {soundWaveAssetPtr=&/ImpulseGrenade_Explosion_Close_03}
 }
 {
  id=WavePlayer_17
  properties {soundWaveAssetPtr=&/ImpulseGrenade_Explosion_Close_04}
 }
 {
  id=WavePlayer_18
  properties {soundWaveAssetPtr=&/ImpulseGrenade_Explosion_Distant_01}
 }
 {
  id=WavePlayer_19
  properties {soundWaveAssetPtr=&/ImpulseGrenade_Explosion_Distant_02}
 }
 {
  id=WavePlayer_20
  properties {soundWaveAssetPtr=&/ImpulseGrenade_Explosion_Distant_03}
 }
 {
  id=WavePlayer_21
  properties {soundWaveAssetPtr=&/ImpulseGrenade_Explosion_Distant_04}
 }
 {
  id=WavePlayer_8
  properties {soundWaveAssetPtr=&/ImpulseGrenade_Explosion_Close_01}
 }
 {
  id=WavePlayer_9
  properties {soundWaveAssetPtr=&/ImpulseGrenade_Explosion_Close_02}
 }
]
properties {
 bOverrideConcurrency=true
 maxDistance=9625
 soundClassObject=&Fort_Gadgets_Items
 volumeMultiplier=1.2
 concurrencyOverrides {resolutionRule=PreventNew}
}
