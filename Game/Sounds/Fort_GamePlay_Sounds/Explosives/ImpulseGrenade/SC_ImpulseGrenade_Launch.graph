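# Sound cue for the impulse grenade launch whoosh, classed as a gadget sound with attenuation and a stop-oldest concurrency cap.
# After a short delay and a smoothed Doppler node, IsEnemy and IsLocal branches set volume: enemy loudest, non-local non-enemy quietest.
# All paths share one mix: delayed, slightly raised train-launch whooshes plus quieter, high-pitched impulse whooshes.
# Quality nodes randomize variants at the first two levels and play single waves at the others.
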SoundCue /Game/Sounds/Fort_GamePlay_Sounds/Explosives/ImpulseGrenade/SC_ImpulseGrenade_Launch
root=Delay_1
nodes [
 {id=Branch_0, children [Mixer_1,Branch_1,Branch_1], properties {boolParameterName=IsEnemy}}
 {id=Branch_1, children [Mixer_2,Mixer_0,Mixer_2], properties {boolParameterName=IsLocal}}
 {id=Delay_0, children [Modulator_4], properties {delayMax=0.2, delayMin=0.2}}
 {id=Delay_1, children [Doppler_0], properties {delayMax=0.1, delayMin=0.1}}
 {id=Doppler_0, children [Branch_0], properties {bUseSmoothing=true, dopplerIntensity=2}}
 {id=Mixer_0, children [Mixer_3], properties {inputVolume [0.3]}}
 {id=Mixer_1, children [Mixer_3], properties {inputVolume [1.1]}}
 {id=Mixer_2, children [Mixer_3], properties {inputVolume [0.85]}}
 {id=Mixer_3, children [Delay_0,Modulator_1], properties {inputVolume [1,0.5]}}
 {id=Modulator_1, children [QualityLevel_1], properties {pitchMax=2, pitchMin=2}}
 {id=Modulator_4, children [QualityLevel_0], properties {pitchMax=1.2, pitchMin=1.1}}
 {id=QualityLevel_0, children [Random_4,Random_4,WavePlayer_3,WavePlayer_3]}
 {id=QualityLevel_1, children [Random_2, Random_2, WavePlayer_15, WavePlayer_15]}
 {id=Random_2, children [WavePlayer_13, WavePlayer_14, WavePlayer_15], properties {weights [1,1,1]}}
 {id=Random_4, children [WavePlayer_3, WavePlayer_4, WavePlayer_5, WavePlayer_12], properties {weights [1,1,1,1]}}
 {id=WavePlayer_12, properties {soundWaveAssetPtr=&/Train_PlayerLaunch_Whoosh_04}}
 {
  id=WavePlayer_13
  properties {soundWaveAssetPtr=&/Grenade_Shockwave_Impulse_Launch_Whoosh_03}
 }
 {
  id=WavePlayer_14
  properties {soundWaveAssetPtr=&/Grenade_Shockwave_Impulse_Launch_Whoosh_02}
 }
 {
  id=WavePlayer_15
  properties {soundWaveAssetPtr=&/Grenade_Shockwave_Impulse_Launch_Whoosh_01}
 }
 {id=WavePlayer_3, properties {soundWaveAssetPtr=&/Train_PlayerLaunch_Whoosh_01}}
 {id=WavePlayer_4, properties {soundWaveAssetPtr=&/Train_PlayerLaunch_Whoosh_02}}
 {id=WavePlayer_5, properties {soundWaveAssetPtr=&/Train_PlayerLaunch_Whoosh_03}}
]
properties {
 attenuationSettings=/Game/Sounds/Attenuation/Fort_Grenade_Shockwave_Impulse_Launch_Att
 bOverrideConcurrency=true
 maxDistance=11000
 soundClassObject=&Fort_Gadgets_Items
 volumeMultiplier=1.3
 attenuationOverrides {distanceAlgorithm=NaturalSound, falloffDistance=3800, lPFRadiusMax=3000, lPFRadiusMin=2000, attenuationShapeExtents {x=2200}}
 concurrencyOverrides {maxCount=2, resolutionRule=StopOldest}
}
