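# Sound cue for picking up an Impulse Grenade, routed to the gadget items sound class.
# A Mixer layers the Shockwave Grenade pickup sound, pitched up through a Modulator, under a louder random pick of three Impulse Grenade drop sounds.
# It uses the shared weapon pickup attenuation and overrides concurrency to block new instances instead of stopping playing ones.
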SoundCue /Game/Sounds/Fort_GamePlay_Sounds/Explosives/ImpulseGrenade/SC_ImpulseGrenade_Pickup
root=Mixer_0
nodes [
 {id=Mixer_0, children [Modulator_0,Random_0], properties {inputVolume [0.4,1.1]}}
 {id=Modulator_0, children [WavePlayer_1], properties {pitchMax=1.1, pitchMin=1.2}}
 {id=Random_0, children [WavePlayer_3,WavePlayer_4,WavePlayer_5], properties {weights [1,1,1]}}
 {id=WavePlayer_1, properties {soundWaveAssetPtr=/Game/Sounds/Fort_GamePlay_Sounds/Explosives/ShockwaveGrenade/Grenade_Shockwave_Pickup01}}
 {id=WavePlayer_3, properties {soundWaveAssetPtr=&/ImpulseGrenade_Drop_01}}
 {id=WavePlayer_4, properties {soundWaveAssetPtr=&/ImpulseGrenade_Drop_02}}
 {id=WavePlayer_5, properties {soundWaveAssetPtr=&/ImpulseGrenade_Drop_03}}
]
properties {
 attenuationSettings=&Fort_Weapon_Pickup_Att
 bOverrideConcurrency=true
 maxDistance=3500
 soundClassObject=&Fort_Gadgets_Items
 volumeMultiplier=0.4
 concurrencyOverrides {resolutionRule=PreventNew}
}
