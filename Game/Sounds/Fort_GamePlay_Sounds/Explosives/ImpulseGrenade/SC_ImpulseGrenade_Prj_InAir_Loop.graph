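# Sound cue for an Impulse Grenade projectile in flight, combining a launch one-shot with a looping in-air sound.
# A mixer blends the loop, faded in by an enveloper volume curve and given smoothed Doppler, with the launch sound, pitch-randomized by a modulator.
# Each branch has its own projectile attenuation asset, and the cue overrides concurrency to cap simultaneous instances.
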SoundCue /Game/Sounds/Fort_GamePlay_Sounds/Explosives/ImpulseGrenade/SC_ImpulseGrenade_Prj_InAir_Loop
root=Mixer_1
nodes [
 {
  id=Attenuation_0
  children [Modulator_0]
  properties {attenuationSettings=/Game/Sounds/Attenuation/Fort_Weapon_Projectile_InAir_Launch_Throw_Att}
 }
 {id=Attenuation_2, children [Enveloper_0], properties {attenuationSettings=/Game/Sounds/Attenuation/Fort_Weapon_Projectile_InAir_Loop_Att}}
 {id=Doppler_0, children [Attenuation_2], properties {bUseSmoothing=true}}
 {
  id=Enveloper_0
  children [WavePlayer_6]
  properties {
   pitchCurve {editorCurveData {keys [{time=0}]}}
   volumeCurve {editorCurveData {keys [{arriveTangent=2.2634, interpMode=RCIM_Cubic, leaveTangent=2.2634, tangentMode=RCTM_User, time=0, value=0}, <at 1 1>]}}
  }
 }
 {id=Mixer_1, children [Doppler_0, Attenuation_0], properties {inputVolume [1.15,1]}}
 {id=Modulator_0, children [WavePlayer_7], properties {pitchMax=1.2, pitchMin=0.9}}
 {
  id=WavePlayer_6
  properties {bLooping=true, soundWaveAssetPtr=&/ImpulseGrenade_Prj_InAir_Loop}
 }
 {
  id=WavePlayer_7
  properties {soundWaveAssetPtr=&/ImpulseGrenade_Prj_InAir_Launch_01}
 }
]
properties {
 bOverrideConcurrency=true
 maxDistance=8000
 soundClassObject=&Fort_WeaponRanged_Fire_3P_PSM
 volumeMultiplier=1.1
 concurrencyOverrides {maxCount=3}
}
