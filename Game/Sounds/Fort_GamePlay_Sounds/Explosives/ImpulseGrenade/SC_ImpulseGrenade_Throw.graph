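# Sound cue for throwing an Impulse Grenade, using gadget sound class and gadget-placement attenuation.
# A mixer layers one of three Impulse Grenade throw waves, randomly pitched up via a modulator, with one of two borrowed Shockwave Grenade throw waves.
# The Shockwave layer plays after a short fixed delay and sits slightly lower in the mix.
# Overridden concurrency prevents new instances once its limit is reached.
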SoundCue /Game/Sounds/Fort_GamePlay_Sounds/Explosives/ImpulseGrenade/SC_ImpulseGrenade_Throw
root=Mixer_2
nodes [
 {id=Delay_0, children [Random_0], properties {delayMax=0.1, delayMin=0.1}}
 {id=Mixer_2, children [Delay_0,Modulator_1], properties {inputVolume [0.7,0.8]}}
 {id=Modulator_1, children [Random_1], properties {pitchMax=1.2, pitchMin=1}}
 {id=Random_0, children [WavePlayer_0,WavePlayer_2], properties {weights [1,1]}}
 {id=Random_1, children [WavePlayer_1,WavePlayer_3,WavePlayer_4], properties {weights [1,1,1]}}
 {id=WavePlayer_0, properties {soundWaveAssetPtr=/Game/Sounds/Fort_GamePlay_Sounds/Explosives/ShockwaveGrenade/Grenade_Shockwave_Throw01}}
 {id=WavePlayer_1, properties {soundWaveAssetPtr=&/ImpulseGrenade_Throw_01}}
 {id=WavePlayer_2, properties {soundWaveAssetPtr=/Game/Sounds/Fort_GamePlay_Sounds/Explosives/ShockwaveGrenade/Grenade_Shockwave_Throw03}}
 {id=WavePlayer_3, properties {soundWaveAssetPtr=&/ImpulseGrenade_Throw_02}}
 {id=WavePlayer_4, properties {soundWaveAssetPtr=&/ImpulseGrenade_Throw_03}}
]
properties {
 attenuationSettings=/Game/Sounds/Attenuation/Fort_GadgetPlace_Att
 bOverrideConcurrency=true
 maxDistance=2000
 soundClassObject=/Game/Sounds/FortSoundClassesAndModes/Fort_Gadgets_StW
 volumeMultiplier=0.4
 concurrencyOverrides {resolutionRule=PreventNew}
}
