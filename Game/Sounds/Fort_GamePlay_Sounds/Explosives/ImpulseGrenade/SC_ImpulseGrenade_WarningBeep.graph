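# Sound cue for the Impulse Grenade warning beep, played as a gadget/item sound.
# A mixer layers a fixed start beep, slightly boosted, with one of three beep variants picked at random with equal weight.
# It routes through the gadgets/items sound class and uses a medium explosive-beep attenuation for spatial falloff.
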SoundCue /Game/Sounds/Fort_GamePlay_Sounds/Explosives/ImpulseGrenade/SC_ImpulseGrenade_WarningBeep
root=Mixer_0
nodes [
 {id=Mixer_0, children [WavePlayer_0,Random_0], properties {inputVolume [1.2,1]}}
 {id=Random_0, children [WavePlayer_1,WavePlayer_2,WavePlayer_3], properties {weights [1,1,1]}}
 {
  id=WavePlayer_0
  properties {soundWaveAssetPtr=&/ImpulseGrenade_WarningBeep_Start_01}
 }
 {id=WavePlayer_1, properties {soundWaveAssetPtr=&/ImpulseGrenade_WarningBeep_01}}
 {id=WavePlayer_2, properties {soundWaveAssetPtr=&/ImpulseGrenade_WarningBeep_02}}
 {id=WavePlayer_3, properties {soundWaveAssetPtr=&/ImpulseGrenade_WarningBeep_03}}
]
properties {
 attenuationSettings=/Game/Sounds/Attenuation/Fort_ExplosiveBeep_Medium_Att
 maxDistance=6500
 soundClassObject=&Fort_Gadgets_Items
 volumeMultiplier=0.9
}
